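# Physics asset for the F_MED_SAM_Dark backpack cosmetic, defining collision bodies and joints for its knives, knife ties and front strap.
# A box covers the pack, capsules cover the dangling parts, and the left knife adds two boxes.
# Both knives swing widely from a knives master, and two five-link tie chains hang from a free-rotating master on the left knife.
# The four-link front strap bends mainly on one axis from its own master, which has no constraint to the pack.
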PhysicsAsset /BRCosmetics/Accessories/FORT_Backpacks/Backpack_F_MED_SAM_Dark/Meshes/F_MED_SAM_Dark_Pack_Physics
bodies {
 dyn_L_knife_jnt [
  {shapeType=Capsule, length=45, radius=2.25, shapeName=dyn_L_knife_jnt_capsule, center=<2.2021 0 0>, rotation=<rot 90 0 0>}
  {shapeType=Box, extentX=7, extentY=5.5, extentZ=5, shapeName=dyn_L_knife_jnt_box, center=<-5.7668 3.6847 3.9732>}
  {
   shapeType=Box
   extentX=7.5
   extentY=7.5
   extentZ=7.5
   shapeName=dyn_L_knife_jnt_box
   center=<1.8199 3.0938 -0.0391>
   rotation=<rot -9.8465 -1.7539 10.1522>
  }
 ]
 dyn_R_knife_jnt [{shapeType=Capsule, length=45.75, radius=2.75, shapeName=dyn_R_knife_jnt_capsule, center=<0 0 -0.5383>, rotation=<rot 90 -0.0002 -0.0005>}]
 dyn_front_strap_1_jnt [{shapeType=Capsule, length=1, radius=1, shapeName=dyn_front_strap_1_jnt_capsule, center=<1.6113 0 0>, rotation=<rot -90 179.9802 -179.9802>}]
 dyn_front_strap_2_jnt [{shapeType=Capsule, length=1, radius=1, shapeName=dyn_front_strap_2_jnt_capsule, center=<1.6113 0 0>, rotation=<rot -90 179.9802 -179.9802>}]
 dyn_front_strap_3_jnt [{shapeType=Capsule, length=1, radius=1, shapeName=dyn_front_strap_3_jnt_capsule, center=<1.6113 0 0>, rotation=<rot -90 179.9802 -179.9802>}]
 dyn_front_strap_4_jnt [{shapeType=Capsule, length=1, radius=1, shapeName=dyn_front_strap_4_jnt_capsule, center=<1.6113 0 0>, rotation=<rot -90 179.9802 -179.9802>}]
 dyn_front_strap_master_jnt [{shapeType=Capsule, length=0.505, radius=0.505, shapeName=dyn_front_strap_master_jnt_capsule}]
 dyn_knife_tie1_1_jnt [{shapeType=Capsule, length=1.5, radius=0.5, shapeName=dyn_knife_tie1_2_jnt_capsule, center=<1.2687 0 0>, rotation=<rot -90 -0.0264 0.0264>}]
 dyn_knife_tie1_2_jnt [{shapeType=Capsule, length=1.5, radius=0.5, shapeName=dyn_knife_tie1_2_jnt_capsule, center=<1.2687 0 0>, rotation=<rot -90 -0.0264 0.0264>}]
 dyn_knife_tie1_3_jnt [{shapeType=Capsule, length=1.5, radius=0.5, shapeName=dyn_knife_tie1_3_jnt_capsule, center=<1.2687 0 0>, rotation=<rot -90 -0.0264 0.0264>}]
 dyn_knife_tie1_4_jnt [{shapeType=Capsule, length=1.5, radius=0.5, shapeName=dyn_knife_tie1_4_jnt_capsule, center=<1.2687 0 0>, rotation=<rot -90 -0.0264 0.0264>}]
 dyn_knife_tie1_5_jnt [{shapeType=Capsule, length=1.5, radius=0.5, shapeName=dyn_knife_tie1_5_jnt_capsule, center=<1.2687 0 0>, rotation=<rot -90 -0.0264 0.0264>}]
 dyn_knife_tie2_1_jnt [{
   shapeType=Capsule
   length=1.5
   radius=0.5
   shapeName=dyn_knife_tie2_2_jnt_capsule
   center=<1.2687 -0.365 0.0001>
   rotation=<rot -90 5.0512 -5.0512>
  }]
 dyn_knife_tie2_2_jnt [{
   shapeType=Capsule
   length=1.5
   radius=0.5
   shapeName=dyn_knife_tie2_2_jnt_capsule
   center=<1.2687 -0.365 0.0001>
   rotation=<rot -90 5.0512 -5.0512>
  }]
 dyn_knife_tie2_3_jnt [{
   shapeType=Capsule
   length=1.5
   radius=0.5
   shapeName=dyn_knife_tie2_3_jnt_capsule
   center=<1.2687 -0.365 0.0001>
   rotation=<rot -90 5.0512 -5.0512>
  }]
 dyn_knife_tie2_4_jnt [{
   shapeType=Capsule
   length=1.5
   radius=0.5
   shapeName=dyn_knife_tie2_4_jnt_capsule
   center=<1.2687 -0.365 0.0001>
   rotation=<rot -90 5.0512 -5.0512>
  }]
 dyn_knife_tie2_5_jnt [{
   shapeType=Capsule
   length=1.5
   radius=0.5
   shapeName=dyn_knife_tie2_5_jnt_capsule
   center=<1.2687 -0.365 0.0001>
   rotation=<rot -90 5.0512 -5.0512>
  }]
 dyn_knife_ties_master_jnt [{shapeType=Capsule, length=0.5, radius=0.5, shapeName=dyn_knife_ties_master_jnt_capsule, rotation=<rot 0 -90 0>}]
 knives_master_jnt [{shapeType=Capsule, length=1, radius=1, shapeName=knives_master_jnt_capsule}]
 pack [{shapeType=Box, extentX=20.5, extentY=7.5, extentZ=24, shapeName=pack_box, center=<-0.0067 -6.4788 -10.6624>}]
}
constraints [
 {bone1Name=dyn_knife_tie2_2_jnt, bone2Name=dyn_knife_tie2_1_jnt, swing1LimitDegrees=25, swing2LimitDegrees=25, twistLimitDegrees=5}
 {bone1Name=dyn_knife_tie2_3_jnt, bone2Name=dyn_knife_tie2_2_jnt, swing1LimitDegrees=25, swing2LimitDegrees=25, twistLimitDegrees=5}
 {bone1Name=dyn_knife_tie2_4_jnt, bone2Name=dyn_knife_tie2_3_jnt, swing1LimitDegrees=25, swing2LimitDegrees=25, twistLimitDegrees=5}
 {bone1Name=dyn_knife_tie2_5_jnt, bone2Name=dyn_knife_tie2_4_jnt, swing1LimitDegrees=25, swing2LimitDegrees=25, twistLimitDegrees=5}
 {bone1Name=dyn_knife_tie1_2_jnt, bone2Name=dyn_knife_tie1_1_jnt, swing1LimitDegrees=25, swing2LimitDegrees=25, twistLimitDegrees=5}
 {bone1Name=dyn_knife_tie1_3_jnt, bone2Name=dyn_knife_tie1_2_jnt, swing1LimitDegrees=25, swing2LimitDegrees=25, twistLimitDegrees=5}
 {bone1Name=dyn_knife_tie1_4_jnt, bone2Name=dyn_knife_tie1_3_jnt, swing1LimitDegrees=25, swing2LimitDegrees=25, twistLimitDegrees=5}
 {bone1Name=dyn_knife_tie1_5_jnt, bone2Name=dyn_knife_tie1_4_jnt, swing1LimitDegrees=25, swing2LimitDegrees=25, twistLimitDegrees=5}
 {bone1Name=dyn_R_knife_jnt, bone2Name=knives_master_jnt, swing1LimitDegrees=90, swing2LimitDegrees=90, twistLimitDegrees=5}
 {bone1Name=dyn_L_knife_jnt, bone2Name=knives_master_jnt, swing1LimitDegrees=90, swing2LimitDegrees=90, twistLimitDegrees=5}
 {
  bone1Name=dyn_knife_ties_master_jnt
  bone2Name=dyn_L_knife_jnt
  swing1LimitDegrees=45
  swing1Motion=Free
  swing2LimitDegrees=45
  swing2Motion=Free
  twistLimitDegrees=45
  twistMotion=Free
 }
 {bone1Name=dyn_front_strap_4_jnt, bone2Name=dyn_front_strap_3_jnt, swing1LimitDegrees=45, swing2LimitDegrees=2, twistLimitDegrees=2}
 {bone1Name=dyn_front_strap_1_jnt, bone2Name=dyn_front_strap_master_jnt, swing1LimitDegrees=45, swing2LimitDegrees=2, twistLimitDegrees=2}
 {bone1Name=dyn_front_strap_2_jnt, bone2Name=dyn_front_strap_1_jnt, swing1LimitDegrees=45, swing2LimitDegrees=2, twistLimitDegrees=2}
 {bone1Name=dyn_front_strap_3_jnt, bone2Name=dyn_front_strap_2_jnt, swing1LimitDegrees=45, swing2LimitDegrees=2, twistLimitDegrees=2}
 {bone1Name=dyn_knife_tie1_1_jnt, bone2Name=dyn_knife_ties_master_jnt, swing1LimitDegrees=25, swing2LimitDegrees=25, twistLimitDegrees=5}
 {bone1Name=dyn_knife_tie2_1_jnt, bone2Name=dyn_knife_ties_master_jnt, swing1LimitDegrees=25, swing2LimitDegrees=25, twistLimitDegrees=5}
]
