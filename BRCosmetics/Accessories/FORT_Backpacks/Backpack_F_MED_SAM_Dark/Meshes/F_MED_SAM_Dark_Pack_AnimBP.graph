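# Animation Blueprint for the F_MED_SAM_Dark backpack, adding secondary physics motion and a skydive-specific pose.
# A RigidBody node simulates in pack-bone space, then additive bone rotations on both knife joints and the pack apply while skydiving or floating.
# Weapon-change and skydive events drive a dynamic material's "Hide Element 01", setting it only when a NinjaWolf-tagged pickaxe is equipped outside skydiving.
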AnimBlueprint /BRCosmetics/Accessories/FORT_Backpacks/Backpack_F_MED_SAM_Dark/Meshes/F_MED_SAM_Dark_Pack_AnimBP
target_skeleton=&/F_MED_SAM_Dark_Pack_Skeleton
graphs [
 AnimGraph {nodes [
   n1 = RigidBody(ComponentPose=Local To Component(LocalPose=Local Space Ref Pose())) {
    bOverrideWorldGravity=true
    overridePhysicsAsset=&/F_MED_SAM_Dark_Pack_Physics
    simulationSpace=BaseBoneSpace
    alphaBoolBlend {blendOption=Linear}
    baseBoneRef {boneName=pack}
    componentAppliedLinearAccClamp=<3000 3000 3000>
    componentLinearAccScale=<0.4 0.3 0.3>
    componentLinearVelScale=<0.4 0.3 0.3>
    overrideWorldGravity {z=-980}
   }
   n2 = "Transform (Modify) Bone - Bone: dyn_R_knife_jnt"(ComponentPose=@n1, Alpha=Get Skydive() -> (Skydive:double), Translation="0,0,0", Rotation="0,-10,0", Scale="1,1,1") {
    rotationMode=BMM_Additive
    rotationSpace=BCS_BoneSpace
    translationSpace=BCS_WorldSpace
    alphaBoolBlend {blendOption=Linear}
    boneToModify {boneName=dyn_R_knife_jnt}
    rotation {yaw=-10}
   }
   n3 = "Transform (Modify) Bone - Bone: dyn_L_knife_jnt"(ComponentPose=@n2, Alpha=Get Skydive(), Translation="0,0,0", Rotation="0,-5,0", Scale="1,1,1") {
    rotationMode=BMM_Additive
    rotationSpace=BCS_BoneSpace
    translationSpace=BCS_WorldSpace
    alphaBoolBlend {blendOption=Linear}
    boneToModify {boneName=dyn_L_knife_jnt}
    rotation {yaw=-5}
   }
   n4 = "Transform (Modify) Bone - Bone: pack"(ComponentPose=@n3, Alpha=Get Skydive(), Translation="0,0,0", Rotation="0,0,15") {
    rotationMode=BMM_Additive
    rotationSpace=BCS_BoneSpace
    translationSpace=BCS_WorldSpace
    alphaBoolBlend {blendOption=Linear}
    boneToModify {boneName=pack}
    rotation {roll=15}
   }
   Output Pose(Result=Component To Local(ComponentPose=@n4)) {layerGroup=None}
  ]}
 EventGraph {nodes [
   Event BlueprintUpdateAnimation()
   n1 = Get bIsSkydiveFloating()
   n2 = Get bIsSkyDiving()
   > Set Skydive(Skydive:double=SelectFloat(A=1, B=0, bPickA=OR Boolean(A=@n2, B=@n1))) -> (Output_Get:double)
   Event OnWeaponChanged()
   > n3 = Get Dynamic Material Instance(SkeletalMeshComponent=GetPartSkeletalMeshComponent())
   n4 = PickaxeItemDefinitionHasAnyMetaTag(Tags=Make Array("[0]"=(TagName="Cosmetics.MetaTags.NinjaWolf")), bRequiresEquipped=true)
   > n5 = Branch(Condition=@n4)
   > n6 = Branch(Condition=Get bIsSkyDiving())
   SetScalarParameterValue(execute=[@n7.then, @n6.then, @n8.else, @n5.else], self=@n3, ParameterName=Hide Element 01, Value=0)
   SetScalarParameterValue(execute=[@n6.else, @n8.then], self=@n3, ParameterName=Hide Element 01, Value=1)
   n7 = Event OnBeginSkydiving()
   Event OnEndSkydiving()
   > n8 = Branch(Condition=@n4)
   TryGetPawnOwner()
  ]}
]
variables [Skydive {type=number}]
